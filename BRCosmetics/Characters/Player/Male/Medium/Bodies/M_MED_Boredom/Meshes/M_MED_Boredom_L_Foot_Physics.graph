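# Physics asset for the Boredom medium male player body that sets up shoelace dynamics; named L_Foot, it contains only right-side bones.
# Inner and outer laces are each three-capsule chains hanging from a small master capsule, with both masters constrained to a large foot_r capsule.
# Lace links use 45-degree swing limits and a near-zero twist limit, while the master-to-foot joints leave swing and twist free.
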PhysicsAsset /BRCosmetics/Characters/Player/Male/Medium/Bodies/M_MED_Boredom/Meshes/M_MED_Boredom_L_Foot_Physics
bodies {
 dyn_R_inner_shoelace_1_jnt [{
   shapeType=Capsule
   length=1.5
   radius=2.5
   shapeName=dyn_R_inner_shoelace_1_jnt_capsule
   center=<-2.1891 0 0.7941>
   rotation=<rot -70 179.9998 -179.9999>
  }]
 dyn_R_inner_shoelace_2_jnt [{
   shapeType=Capsule
   length=1.5
   radius=2.5
   shapeName=dyn_R_inner_shoelace_2_jnt_capsule
   center=<-2.4782 0 0>
   rotation=<rot -90 -0.0005 0.0006>
  }]
 dyn_R_inner_shoelace_3_jnt [{
   shapeType=Capsule
   length=1.5
   radius=2.5
   shapeName=dyn_R_inner_shoelace_3_jnt_capsule
   center=<-2.3676 0.6273 0>
   rotation=<rot -90 0 -9.9998>
  }]
 dyn_R_outer_shoelace_1_jnt [{
   shapeType=Capsule
   length=1.75
   radius=2.5
   shapeName=dyn_R_outer_shoelace_1_jnt_capsule
   center=<-2.3897 0.5142 -0.4214>
   rotation=<rot -79.9999 -0.0012 -9.9989>
  }]
 dyn_R_outer_shoelace_2_jnt [{
   shapeType=Capsule
   length=1.75
   radius=2.5
   shapeName=dyn_R_outer_shoelace_2_jnt_capsule
   center=<-3.0695 0 -0.5412>
   rotation=<rot -79.9999 -0.0003 0.0001>
  }]
 dyn_R_outer_shoelace_3_jnt [{
   shapeType=Capsule
   length=1.75
   radius=2.5
   shapeName=dyn_R_outer_shoelace_3_jnt_capsule
   center=<-3.0695 0 -0.5412>
   rotation=<rot -79.9999 -0.0002 0>
  }]
 foot_r [{shapeType=Capsule, length=20, radius=7.5, shapeName=foot_r_capsule, center=<3.5443 -6.3978 0>, rotation=<rot 0 -0.0002 90>}]
 master_R_inner_shoelace_jnt [{shapeType=Capsule, length=1, radius=1, shapeName=master_R_inner_shoelace_jnt_capsule}]
 master_R_outer_shoelace_jnt [{shapeType=Capsule, length=1, radius=1, shapeName=master_R_outer_shoelace_jnt_capsule}]
}
constraints [
 {
  bone1Name=dyn_R_inner_shoelace_1_jnt
  bone2Name=master_R_inner_shoelace_jnt
  swing1LimitDegrees=45
  swing2LimitDegrees=45
  twistLimitDegrees=1
 }
 {
  bone1Name=dyn_R_inner_shoelace_2_jnt
  bone2Name=dyn_R_inner_shoelace_1_jnt
  swing1LimitDegrees=45
  swing2LimitDegrees=45
  twistLimitDegrees=1
 }
 {
  bone1Name=dyn_R_inner_shoelace_3_jnt
  bone2Name=dyn_R_inner_shoelace_2_jnt
  swing1LimitDegrees=45
  swing2LimitDegrees=45
  twistLimitDegrees=1
 }
 {
  bone1Name=dyn_R_outer_shoelace_3_jnt
  bone2Name=dyn_R_outer_shoelace_2_jnt
  swing1LimitDegrees=45
  swing2LimitDegrees=45
  twistLimitDegrees=1
 }
 {
  bone1Name=dyn_R_outer_shoelace_1_jnt
  bone2Name=master_R_outer_shoelace_jnt
  swing1LimitDegrees=45
  swing2LimitDegrees=45
  twistLimitDegrees=1
 }
 {
  bone1Name=dyn_R_outer_shoelace_2_jnt
  bone2Name=dyn_R_outer_shoelace_1_jnt
  swing1LimitDegrees=45
  swing2LimitDegrees=45
  twistLimitDegrees=1
 }
 {
  bone1Name=master_R_outer_shoelace_jnt
  bone2Name=foot_r
  swing1LimitDegrees=45
  swing1Motion=Free
  swing2LimitDegrees=45
  swing2Motion=Free
  twistLimitDegrees=45
  twistMotion=Free
 }
 {
  bone1Name=master_R_inner_shoelace_jnt
  bone2Name=foot_r
  swing1LimitDegrees=45
  swing1Motion=Free
  swing2LimitDegrees=45
  swing2Motion=Free
  twistLimitDegrees=45
  twistMotion=Free
 }
]
