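# Physics asset for the M_MED_Boredom player body cosmetic, set up for a dangling CD player and its cord at the hip.
# The CD player box and a four-capsule cord chain hang by limited swing and near-rigid twist from master joints, which connect to the pelvis with free rotation.
# A free-rotation constraint ties the last cord loop to the player, closing the chain.
# Pelvis and left thigh capsules are the only body shapes.
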PhysicsAsset /BRCosmetics/Characters/Player/Male/Medium/Bodies/M_MED_Boredom/Meshes/M_MED_Boredom_Physics
bodies {
 dyn_CDplayer_jnt [{shapeType=Box, extentX=15, extentY=16, extentZ=7.5, shapeName=dyn_CDplayer_jnt_box, center=<8.4149 0.9877 3.471>}]
 dyn_cord_1_jnt [{shapeType=Capsule, length=1, radius=1, shapeName=dyn_cord_1_jnt_capsule, center=<1.4728 0 0.2597>, rotation=<rot -80.0002 0.0013 -0.0012>}]
 dyn_cord_2_jnt [{
   shapeType=Capsule
   length=1.25
   radius=1
   shapeName=dyn_cord_2_jnt_capsule
   center=<2.3742 -0.1614 0.4186>
   rotation=<rot -80.0004 0.0011 -10.001>
  }]
 dyn_cord_loop_1_jnt [{shapeType=Capsule, length=4.5, radius=2.75, shapeName=dyn_cord_loop_1_jnt_capsule, center=<2.2728 0 0>, rotation=<rot -0.0001 0.0001 -60>}]
 dyn_cord_loop_2_jnt [{shapeType=Capsule, length=4.5, radius=2.75, shapeName=dyn_cord_loop_2_jnt_capsule, center=<2.2728 0 0>, rotation=<rot -0.0001 0.0001 -60>}]
 master_CDplayer_jnt [{shapeType=Capsule, length=1, radius=1, shapeName=master_CDplayer_jnt_capsule, center=<0.7033 0 0>, rotation=<rot 90 0 0>}]
 master_cord_jnt [{shapeType=Capsule, length=1, radius=1, shapeName=master_cord_jnt_capsule}]
 pelvis [{shapeType=Capsule, length=8.5, radius=10, shapeName=pelvis_capsule, center=<-4.833 0 0>}]
 thigh_twist_01_l [{
   shapeType=Capsule
   length=25.6109
   radius=8.77
   shapeName=thigh_twist_01_l_capsule
   center=<2.4387 0.138 3.1864>
   rotation=<rot -3.9121 -90.0002 -99.0034>
  }]
}
constraints [
 {bone1Name=dyn_cord_loop_2_jnt, bone2Name=dyn_cord_loop_1_jnt, swing1LimitDegrees=10, swing2LimitDegrees=35, twistLimitDegrees=1}
 {bone1Name=dyn_cord_loop_1_jnt, bone2Name=dyn_cord_2_jnt, swing1LimitDegrees=10, swing2LimitDegrees=35, twistLimitDegrees=1}
 {bone1Name=dyn_cord_2_jnt, bone2Name=dyn_cord_1_jnt, swing1LimitDegrees=50, swing2LimitDegrees=10, twistLimitDegrees=1}
 {bone1Name=dyn_cord_1_jnt, bone2Name=master_cord_jnt, swing1LimitDegrees=50, swing2LimitDegrees=10, twistLimitDegrees=1}
 {
  bone1Name=thigh_twist_01_l
  bone2Name=pelvis
  swing1LimitDegrees=45
  swing1Motion=Free
  swing2LimitDegrees=45
  swing2Motion=Free
  twistLimitDegrees=45
  twistMotion=Free
 }
 {
  bone1Name=master_CDplayer_jnt
  bone2Name=pelvis
  swing1LimitDegrees=45
  swing1Motion=Free
  swing2LimitDegrees=45
  swing2Motion=Free
  twistLimitDegrees=45
  twistMotion=Free
 }
 {
  bone1Name=master_cord_jnt
  bone2Name=pelvis
  swing1LimitDegrees=45
  swing1Motion=Free
  swing2LimitDegrees=45
  swing2Motion=Free
  twistLimitDegrees=45
  twistMotion=Free
 }
 {bone1Name=dyn_CDplayer_jnt, bone2Name=master_CDplayer_jnt, swing1LimitDegrees=20, swing2LimitDegrees=65, twistLimitDegrees=2}
 {
  bone1Name=dyn_cord_loop_2_jnt
  bone2Name=dyn_CDplayer_jnt
  swing1LimitDegrees=45
  swing1Motion=Free
  swing2LimitDegrees=45
  swing2Motion=Free
  twistLimitDegrees=45
  twistMotion=Free
 }
]
