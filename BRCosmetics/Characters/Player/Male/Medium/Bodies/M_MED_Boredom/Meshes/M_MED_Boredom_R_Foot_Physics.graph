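# Physics asset for the M_MED_Boredom cosmetic body's foot mesh, set up for shoelace physics; despite its R_Foot name, every bone is left-side.
# A large foot_l capsule anchors inner and outer lace chains, each a small master capsule followed by three dyn capsules.
# Master joints rotate freely against the foot, while constraints down each chain limit swing and allow almost no twist.
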PhysicsAsset /BRCosmetics/Characters/Player/Male/Medium/Bodies/M_MED_Boredom/Meshes/M_MED_Boredom_R_Foot_Physics
bodies {
 dyn_L_inner_shoelace_1_jnt [{
   shapeType=Capsule
   length=1.75
   radius=2.5
   shapeName=dyn_L_inner_shoelace_1_jnt_capsule
   center=<2.0161 0 -0.3555>
   rotation=<rot 0.0001 -90.0001 -80.0001>
  }]
 dyn_L_inner_shoelace_2_jnt [{
   shapeType=Capsule
   length=1.75
   radius=2.5
   shapeName=dyn_L_inner_shoelace_2_jnt_capsule
   center=<2.5284 -0.1585 -0.3014>
   rotation=<rot 0 -91.8674 -87.2305>
  }]
 dyn_L_inner_shoelace_3_jnt [{
   shapeType=Capsule
   length=1.75
   radius=2.5
   shapeName=dyn_L_inner_shoelace_3_jnt_capsule
   center=<4.0536 0.4506 0.7431>
   rotation=<rot 0 111.4593 -77.6641>
  }]
 dyn_L_outer_shoelace_1_jnt [{
   shapeType=Capsule
   length=1.75
   radius=2.5
   shapeName=dyn_L_outer_shoelace_1_jnt_capsule
   center=<2.4329 0.3841 -1.0985>
   rotation=<rot 0 -76.1693 -66.0737>
  }]
 dyn_L_outer_shoelace_2_jnt [{
   shapeType=Capsule
   length=1.75
   radius=2.5
   shapeName=dyn_L_outer_shoelace_2_jnt_capsule
   center=<2.9281 0.0594 0.0373>
   rotation=<rot 0 -90.3854 -89.7085>
  }]
 dyn_L_outer_shoelace_3_jnt [{
   shapeType=Capsule
   length=1.75
   radius=2.5
   shapeName=dyn_L_outer_shoelace_3_jnt_capsule
   center=<3.3999 -0.9 0.2275>
   rotation=<rot 0 68.0509 -85.2533>
  }]
 foot_l [{shapeType=Capsule, length=20, radius=7.5, shapeName=foot_l_capsule, center=<-3.2145 6.775 0>, rotation=<rot 0 0 90>}]
 master_L_inner_shoelace_jnt [{shapeType=Capsule, length=0.505, radius=0.505, shapeName=master_L_inner_shoelace_jnt_capsule, rotation=<rot 0 -90 0>}]
 master_L_outer_shoelace_jnt [{shapeType=Capsule, length=0.505, radius=0.505, shapeName=master_L_outer_shoelace_jnt_capsule, rotation=<rot 0 -90 0>}]
}
constraints [
 {
  bone1Name=master_L_inner_shoelace_jnt
  bone2Name=foot_l
  swing1LimitDegrees=45
  swing1Motion=Free
  swing2LimitDegrees=45
  swing2Motion=Free
  twistLimitDegrees=45
  twistMotion=Free
 }
 {
  bone1Name=dyn_L_inner_shoelace_1_jnt
  bone2Name=master_L_inner_shoelace_jnt
  swing1LimitDegrees=45
  swing2LimitDegrees=45
  twistLimitDegrees=1
 }
 {
  bone1Name=dyn_L_inner_shoelace_2_jnt
  bone2Name=dyn_L_inner_shoelace_1_jnt
  swing1LimitDegrees=45
  swing2LimitDegrees=45
  twistLimitDegrees=1
 }
 {
  bone1Name=dyn_L_inner_shoelace_3_jnt
  bone2Name=dyn_L_inner_shoelace_2_jnt
  swing1LimitDegrees=45
  swing2LimitDegrees=45
  twistLimitDegrees=1
 }
 {
  bone1Name=master_L_outer_shoelace_jnt
  bone2Name=foot_l
  swing1LimitDegrees=45
  swing1Motion=Free
  swing2LimitDegrees=45
  swing2Motion=Free
  twistLimitDegrees=45
  twistMotion=Free
 }
 {
  bone1Name=dyn_L_outer_shoelace_1_jnt
  bone2Name=master_L_outer_shoelace_jnt
  swing1LimitDegrees=45
  swing2LimitDegrees=45
  twistLimitDegrees=1
 }
 {
  bone1Name=dyn_L_outer_shoelace_2_jnt
  bone2Name=dyn_L_outer_shoelace_1_jnt
  swing1LimitDegrees=45
  swing2LimitDegrees=45
  twistLimitDegrees=1
 }
 {
  bone1Name=dyn_L_outer_shoelace_3_jnt
  bone2Name=dyn_L_outer_shoelace_2_jnt
  swing1LimitDegrees=45
  swing2LimitDegrees=45
  twistLimitDegrees=1
 }
]
