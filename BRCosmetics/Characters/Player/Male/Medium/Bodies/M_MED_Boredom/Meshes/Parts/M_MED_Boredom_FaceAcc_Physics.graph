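# Physics asset for the face accessory part of the male medium Boredom cosmetic body, covering its hair, hat and cord.
# Capsules on upper spine, clavicles, neck and head form the body, with a master hair joint on the head branching to bangs, cord, back hair and hat masters.
# Those anchor links swing and twist freely, while the dynamic joint chains below them have per-joint swing limits and very tight twist limits.
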PhysicsAsset /BRCosmetics/Characters/Player/Male/Medium/Bodies/M_MED_Boredom/Meshes/Parts/M_MED_Boredom_FaceAcc_Physics
bodies {
 clavicle_l [{shapeType=Capsule, length=9.5, radius=7, shapeName=clavicle_l_capsule, center=<10.5981 0 1.1698>, rotation=<rot -90 92.894 -92.8945>}]
 clavicle_r [{shapeType=Capsule, length=9.5, radius=7, shapeName=clavicle_r_capsule, center=<-11.0459 0 -1.293>, rotation=<rot -90 92.894 -92.8945>}]
 dyn_C_hair_back_1_jnt [{
   shapeType=Capsule
   length=3.75
   radius=2.75
   shapeName=dyn_C_hair_back_1_jnt_capsule
   center=<0.8841 -0.2776 1.7535>
   rotation=<rot -0.001 -9.9997 -90.0002>
  }]
 dyn_C_hair_back_2_jnt [{
   shapeType=Capsule
   length=2.25
   radius=2.75
   shapeName=dyn_C_hair_back_2_jnt_capsule
   center=<1.6372 0.3375 1.9139>
   rotation=<rot -0.0007 0.0001 -80.0004>
  }]
 dyn_L_bangflip_jnt [{
   shapeType=Capsule
   length=2.5
   radius=2.5
   shapeName=dyn_L_bangflip_jnt_capsule
   center=<3.9093 0 -1.4229>
   rotation=<rot 69.9999 0.0001 -0.0001>
  }]
 dyn_L_bangsside_jnt [{
   shapeType=Capsule
   length=2.25
   radius=2.75
   shapeName=dyn_L_bangsside_jnt_capsule
   center=<4.5393 2.5095 -0.2111>
   rotation=<rot -80 -179.9998 180>
  }]
 dyn_L_hair_back_jnt [{
   shapeType=Capsule
   length=3.25
   radius=3.25
   shapeName=dyn_L_hair_back_jnt_capsule
   center=<2.3824 -0.2292 1.2997>
   rotation=<rot -0.0001 0 -99.9999>
  }]
 dyn_R_bangflip_1_jnt [{
   shapeType=Capsule
   length=2
   radius=2
   shapeName=dyn_R_bangflip_1_jnt_capsule
   center=<2.0378 0.5692 0>
   rotation=<rot -90 -179.9999 -179.9998>
  }]
 dyn_R_bangflip_2_jnt [{shapeType=Capsule, length=2, radius=2, shapeName=dyn_R_bangflip_2_jnt_capsule, center=<2.5357 0.6029 0>, rotation=<rot -90 -90 110.0004>}]
 dyn_R_bangsside_jnt [{
   shapeType=Capsule
   length=2
   radius=2
   shapeName=dyn_R_bangsside_jnt_capsule
   center=<1.2729 2.3517 -0.0502>
   rotation=<rot -90 -172.875 -127.1248>
  }]
 dyn_R_hair_back_jnt [{
   shapeType=Capsule
   length=3.25
   radius=3
   shapeName=dyn_R_hair_back_jnt_capsule
   center=<2.7155 0.8424 1.3377>
   rotation=<rot -0.0001 0 -90.0001>
  }]
 dyn_cord_1_jnt [{shapeType=Capsule, length=1.25, radius=2.75, shapeName=dyn_cord_1_jnt_capsule, center=<1.7092 0.7268 0>, rotation=<rot -90 180 -159.9996>}]
 dyn_cord_2_jnt [{shapeType=Capsule, length=1.25, radius=2.75, shapeName=dyn_cord_2_jnt_capsule, center=<1.8005 2.1663 0.2189>, rotation=<rot -70 -9.9997 10>}]
 dyn_hat_movement_jnt [{shapeType=Capsule, length=2.75, radius=8.75, shapeName=dyn_hat_movement_jnt_capsule, center=<0 0 -0.3654>, rotation=<rot -90 -90 90>}]
 dyn_hat_puff_jnt [{shapeType=Capsule, length=1, radius=4.5, shapeName=dyn_hat_puff_jnt_capsule, center=<0 0 3.5906>, rotation=<rot -0.0003 0 0>}]
 dyn_mainbangs_1_jnt [{
   shapeType=Capsule
   length=4.75
   radius=3.25
   shapeName=dyn_mainbangs_1_jnt_capsule
   center=<3.1217 1.6983 -0.5504>
   rotation=<rot -0.0001 -179.9985 179.9997>
  }]
 dyn_mainbangs_2_jnt [{
   shapeType=Capsule
   length=1
   radius=4.75
   shapeName=dyn_mainbangs_2_jnt_capsule
   center=<4.1762 2.9039 0.6706>
   rotation=<rot -80 -179.9986 179.9998>
  }]
 dyn_mainbangs_3_jnt [{
   shapeType=Capsule
   length=3.5
   radius=3.75
   shapeName=dyn_mainbangs_3_jnt_capsule
   center=<5.4527 1.5715 -0.9615>
   rotation=<rot -80.0001 -179.9991 179.9999>
  }]
 head [{shapeType=Capsule, length=6.5, radius=9, shapeName=head_capsule, center=<4.2195 7.4029 0>, rotation=<rot 90 0 0>}]
 master_bangs_jnt [{shapeType=Capsule, length=0.505, radius=0.505, shapeName=master_bangs_jnt_capsule, rotation=<rot 0 -90 0>}]
 master_cord_jnt [{shapeType=Capsule, length=1, radius=1, shapeName=master_cord_jnt_capsule}]
 master_hair_back_jnt [{shapeType=Capsule, length=2, radius=2, shapeName=master_hair_back_jnt_capsule}]
 master_hair_jnt [{shapeType=Capsule, length=19.25, radius=4.5, shapeName=master_hair_jnt_capsule, center=<4.6517 5.8894 0>}]
 master_hat_back [{shapeType=Capsule, length=2, radius=2, shapeName=master_hat_back_capsule}]
 neck_02 [{shapeType=Capsule, length=6.5, radius=5.75, shapeName=neck_02_capsule, center=<-2.5639 2.4102 0>, rotation=<rot 90 0 -10>}]
 spine_05 [{shapeType=Capsule, length=13, radius=11.25, shapeName=spine_05_capsule, center=<1.1497 -5.1749 0>}]
}
constraints [
 {bone1Name=dyn_R_bangflip_1_jnt, bone2Name=dyn_mainbangs_1_jnt, swing1LimitDegrees=5, swing2LimitDegrees=5, twistLimitDegrees=2}
 {bone1Name=dyn_R_bangflip_2_jnt, bone2Name=dyn_R_bangflip_1_jnt, swing1LimitDegrees=5, swing2LimitDegrees=5, twistLimitDegrees=2}
 {bone1Name=dyn_R_bangsside_jnt, bone2Name=dyn_mainbangs_1_jnt, swing1LimitDegrees=25, swing2LimitDegrees=25, twistLimitDegrees=2}
 {bone1Name=dyn_mainbangs_2_jnt, bone2Name=dyn_mainbangs_1_jnt, swing1LimitDegrees=25, swing2LimitDegrees=10, twistLimitDegrees=2}
 {bone1Name=dyn_mainbangs_3_jnt, bone2Name=dyn_mainbangs_2_jnt, swing1LimitDegrees=25, swing2LimitDegrees=10, twistLimitDegrees=2}
 {bone1Name=dyn_L_bangsside_jnt, bone2Name=master_bangs_jnt, swing1LimitDegrees=25, swing2LimitDegrees=10, twistLimitDegrees=2}
 {bone1Name=dyn_mainbangs_1_jnt, bone2Name=master_bangs_jnt, swing1LimitDegrees=25, swing2LimitDegrees=10, twistLimitDegrees=2}
 {bone1Name=dyn_L_bangflip_jnt, bone2Name=master_bangs_jnt, swing1LimitDegrees=25, swing2LimitDegrees=25, twistLimitDegrees=2}
 {
  bone1Name=master_bangs_jnt
  bone2Name=master_hair_jnt
  swing1LimitDegrees=45
  swing1Motion=Free
  swing2LimitDegrees=45
  swing2Motion=Free
  twistLimitDegrees=45
  twistMotion=Free
 }
 {
  bone1Name=master_cord_jnt
  bone2Name=master_hair_jnt
  swing1LimitDegrees=45
  swing1Motion=Free
  swing2LimitDegrees=45
  swing2Motion=Free
  twistLimitDegrees=45
  twistMotion=Free
 }
 {bone1Name=dyn_cord_1_jnt, bone2Name=master_cord_jnt, swing1LimitDegrees=5, swing2LimitDegrees=15, twistLimitDegrees=2}
 {bone1Name=dyn_cord_2_jnt, bone2Name=dyn_cord_1_jnt, swing1LimitDegrees=5, swing2LimitDegrees=15, twistLimitDegrees=2}
 {
  bone1Name=master_hair_back_jnt
  bone2Name=master_hair_jnt
  swing1LimitDegrees=45
  swing1Motion=Free
  swing2LimitDegrees=45
  swing2Motion=Free
  twistLimitDegrees=45
  twistMotion=Free
 }
 {bone1Name=dyn_C_hair_back_1_jnt, bone2Name=master_hair_back_jnt, swing1LimitDegrees=45, swing2LimitDegrees=45, twistLimitDegrees=2}
 {bone1Name=dyn_C_hair_back_2_jnt, bone2Name=dyn_C_hair_back_1_jnt, swing1LimitDegrees=45, swing2LimitDegrees=45, twistLimitDegrees=2}
 {bone1Name=dyn_L_hair_back_jnt, bone2Name=master_hair_back_jnt, swing1LimitDegrees=25, swing2LimitDegrees=60, twistLimitDegrees=2}
 {bone1Name=dyn_R_hair_back_jnt, bone2Name=master_hair_back_jnt, swing1LimitDegrees=25, swing2LimitDegrees=15, twistLimitDegrees=2}
 {
  bone1Name=master_hat_back
  bone2Name=master_hair_jnt
  swing1LimitDegrees=45
  swing1Motion=Free
  swing2LimitDegrees=45
  swing2Motion=Free
  twistLimitDegrees=45
  twistMotion=Free
 }
 {bone1Name=dyn_hat_movement_jnt, bone2Name=master_hat_back, swing1LimitDegrees=5, swing2LimitDegrees=5, twistLimitDegrees=1}
 {bone1Name=dyn_hat_puff_jnt, bone2Name=dyn_hat_movement_jnt, swing1LimitDegrees=50, swing2LimitDegrees=50, twistLimitDegrees=1}
 {
  bone1Name=master_hair_jnt
  bone2Name=head
  swing1LimitDegrees=45
  swing1Motion=Free
  swing2LimitDegrees=45
  swing2Motion=Free
  twistLimitDegrees=45
  twistMotion=Free
 }
 {
  bone1Name=head
  bone2Name=neck_02
  swing1LimitDegrees=45
  swing1Motion=Free
  swing2LimitDegrees=45
  swing2Motion=Free
  twistLimitDegrees=45
  twistMotion=Free
 }
 {
  bone1Name=neck_02
  bone2Name=spine_05
  swing1LimitDegrees=45
  swing1Motion=Free
  swing2LimitDegrees=45
  swing2Motion=Free
  twistLimitDegrees=45
  twistMotion=Free
 }
 {
  bone1Name=clavicle_l
  bone2Name=spine_05
  swing1LimitDegrees=45
  swing1Motion=Free
  swing2LimitDegrees=45
  swing2Motion=Free
  twistLimitDegrees=45
  twistMotion=Free
 }
 {
  bone1Name=clavicle_r
  bone2Name=spine_05
  swing1LimitDegrees=45
  swing1Motion=Free
  swing2LimitDegrees=45
  swing2Motion=Free
  twistLimitDegrees=45
  twistMotion=Free
 }
]
