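# Animation blueprint for the face accessory part of the M_MED_Boredom male character, adding physics secondary motion.
# It copies the pose from the attached parent mesh, then runs a RigidBody simulation in head bone space with an override physics asset, custom gravity and damped, clamped component motion.
# Each update maps head-to-spine_05 yaw and pitch deltas from ref pose into two clamped variables, which the shown AnimGraph does not read.
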AnimBlueprint /BRCosmetics/Characters/Player/Male/Medium/Bodies/M_MED_Boredom/Meshes/Parts/M_MED_Boredom_FaceAcc_AnimBP
target_skeleton=&/M_MED_Boredom_FaceAcc_Skeleton
graphs [
 AnimGraph {nodes [
   n1 = RigidBody(ComponentPose=Local To Component(LocalPose=Copy Pose From Mesh() {bUseAttachedParent=true})) {
    bClampLinearTranslationLimitToRefPose=true
    bOverrideWorldGravity=true
    overridePhysicsAsset=&/M_MED_Boredom_FaceAcc_Physics
    simulationSpace=BaseBoneSpace
    alphaBoolBlend {blendOption=Linear}
    baseBoneRef {boneName=head}
    componentAppliedLinearAccClamp=<3000 3000 3000>
    componentLinearAccScale=<0.3 0.3 0.2>
    componentLinearVelScale=<0.3 0.3 0.2>
    overrideWorldGravity {z=-980}
    simSpaceSettings {maxAngularAcceleration=1000, maxAngularVelocity=3, maxLinearAcceleration=250, maxLinearVelocity=500, velocityScaleZ=0, worldAlpha=1}
   }
   Output Pose(Result=Component To Local(ComponentPose=@n1)) {layerGroup=None}
  ]}
 EventGraph {nodes [
   Event BlueprintUpdateAnimation()
   n1 = GetOwnerSkeletalMeshComponent()
   n2 = BreakRotator(InRot=BreakTransform(InTransform=GetDeltaTransformFromRefPose(self=@n1, BoneName=head, BaseName=spine_05)).Rotation)
   > Set Head Back Rot(Head Back Rot:double=MapRangeClamped(Value=@n2.Yaw, InRangeA=-10, InRangeB=-55)) -> (Output_Get:double)
   > Set Head R Rot(Head R Rot:double=MapRangeClamped(Value=@n2.Pitch, InRangeA=0, InRangeB=45)) -> (Output_Get:double)
  ]}
]
variables [
 Head Back Rot {type=number}
 Head R Rot {type=number}
]
